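# QuestEditorLauncherAction is an editor utility blueprint that launches the NeverCook Quest Editor tool.
# Its OpenQuestEditor function spawns and registers the QuestEditor utility widget as a tab through the EditorUtilitySubsystem.
# It then casts the resulting widget to QuestEditor and calls OpenSelectedAssets on it, which by its name loads the currently selected assets into the editor.
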EditorUtilityBlueprint /Game/NeverCook/EditorUtilities/QuestTools/SupportWidget/QuestEditorLauncherAction
graphs [
 EventGraph {links [], nodes []}
 OpenQuestEditor {nodes [
   OpenQuestEditor() {is=title K2Node_FunctionEntry}
   > n1 = SpawnAndRegisterTab(self=Get EditorUtilitySubsystem(), InBlueprint=/Game/NeverCook/EditorUtilities/QuestTools/QuestEditor/QuestEditor)
   > n2 = Cast To QuestEditor!(execute, Object=@n1) -> (then, CastFailed, AsQuest Editor:Quest Editor Object Reference) {is=title K2Node_DynamicCast}
   > OpenSelectedAssets(self:Quest Editor Base Object Reference=@n2)
  ]}
]
variables []
